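# Material function meant to read top-down isometric world data for water flow, water and scene height, depth, and obstacle avoidance.
# It projects position to UVs using the iso render pivot and width from FortniteMaterialParameters.
# Both texture samples are disconnected, so constant vectors stand in for the sampled data feeding the outputs.
# It also outputs a World Position Offset that snaps particles slightly above the water height.
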MaterialFunction /Game/WaterAndWind/MaterialFunctions/Fortnite_SampleWindWaterIsometricWorldData
graph {
 properties {bExposeToLibrary=true, libraryCategoriesText [Fortnite Wind]}
 nodes [
  n1 = Constant()
  n2 = MaterialFunctionCall(Float4=Constant4Vector().RGBA) {
   desc=These were plugged into the Sample above
   materialFunction=&BreakOutFloat4Components
  }
  n3 = Subtract(
   A=@n2.B
   B=MaterialFunctionCall(Float3=WorldPosition().XYZ) {materialFunction=&BreakOutFloat3Components}.B
  )
  FunctionOutput(None=MaterialFunctionCall(X=@n1, Y=@n1, Z=Add(A=@n3) {constB=3, desc=water z plane offset}) {materialFunction=&MakeFloat3}.Result) {outputName="Align particles to water surface (Add to WPO)", sortPriority=-1}
  n4 = MaterialFunctionCall(Float4=Constant4Vector().RGBA) {
   desc=These were plugged into the Sample below
   materialFunction=&BreakOutFloat4Components
  }
  FunctionOutput(None=AppendVector(A=@n4.R, B=@n4.R)) {outputName=Obstacle Avoidance Vector RG, sortPriority=7}
  FunctionOutput() {outputName="-------------  ", sortPriority=6}
  n5 = CollectionParameter() {
   collection=&FortniteMaterialParameters
   parameterName=World Iso Render Pivot Point
  }
  n6 = Subtract(A=ComponentMask(None=@n5) {b=true}, B=@n2.A)
  FunctionOutput(None=@n6) {outputName=Scene Z Height, sortPriority=1}
  FunctionOutput() {outputName=-------------}
  FunctionOutput() {outputName="------------- ", sortPriority=4}
  FunctionOutput(None=Subtract(A=@n2.B, B=@n6)) {outputName=Water Depth, sortPriority=3}
  FunctionOutput(None=ConstantBiasScale(None=AppendVector(A=@n2.R, B=@n2.R)) {bias=-0.5, scale=2}) {outputName=Flow Map RG, sortPriority=5}
  FunctionOutput(None=@n2.B) {outputName=Water Z height, sortPriority=2}
  FunctionOutput(None=@n4.B) {outputName="Obstacle Avoidance Fade Off (scalar)", sortPriority=8}
  n7 = Subtract(
   A=FunctionInput(Preview=WorldPosition().XYZ) {bUsePreviewValueAsDefault=true, inputName=Position}
   B=ComponentMask(None=@n5) {b=true, g=true, r=true}
  )
  n8 = CollectionParameter() {collection=&FortniteMaterialParameters, parameterName=World Iso Render Width}
  n9 = Add(A=Divide(A=ComponentMask(None=@n7) {g=true, r=true}, B=@n8)) {constB=0.5}
  TextureSample(UVs=@n9)
  TextureSample(UVs=@n9)
 ]
}
