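# Material function in the Fortnite Wind and Water library providing helper outputs for wind-vector particle materials.
# A collection parameter supplies a Z offset, output as vertical WPO and subtracted from pixel world Z in the adjusted position output, whose X and Y both take world X.
# Sampling isometric wind/water world data at the input position yields ground height and cliff occlusion, the latter from scaled, clamped, inverted obstacle-avoidance fade.
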MaterialFunction /Game/WaterAndWind/MaterialFunctions/Fortnite_WindVectorParticleHelperFunctions
graph {
 properties {bExposeToLibrary=true, libraryCategoriesText [Fortnite Wind and Water]}
 nodes [
  n1 = MaterialFunctionCall(Float3=WorldPosition().XYZ) {materialFunction=&BreakOutFloat3Components}
  n2 = CollectionParameter() {
   collection=&FortniteMaterialParameters
   parameterName=WindVectorParticleOffset
  }
  FunctionOutput(None=MaterialFunctionCall(X=@n1.R, Y=@n1.R, Z=Subtract(A=@n1.B, B=@n2)) {materialFunction=&MakeFloat3}.Result) {outputName=Pixel Shader Particle World Position - Adjusted For Offset}
  n3 = MaterialFunctionCall!(Position=FunctionInput(Preview=WorldPosition().XYZ) {bUsePreviewValueAsDefault=true, inputName=World position}) {materialFunction=&/Fortnite_SampleWindWaterIsometricWorldData}
  n4 = Multiply(A=@n3."Obstacle Avoidance Fade Off (scalar)") {constB=3}
  FunctionOutput(None=OneMinus(None=Clamp(None=@n4))) {outputName=Distance Occlusion To Cliffs}
  n5 = Constant()
  FunctionOutput(None=MaterialFunctionCall(X=@n5, Y=@n5, Z=@n2) {materialFunction=&MakeFloat3}.Result) {outputName=Wind Particle WPO, sortPriority=-5}
  FunctionOutput(None=@n3.Scene Z Height) {outputName=Ground Z height}
 ]
}
